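# Physics asset for the Toxic Kitty backpack cosmetic, defining collision bodies and joints for its dangling batons, mask and straps.
# Two two-segment capsule baton chains share one master body, while the mask and each strap hang from their own small box anchors.
# Baton joints have wide swing and tight twist, straps bend on one swing axis, and the mask rotates freely.
# A capsule-plus-box body covers the pack, with no constraint to the other bodies.
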PhysicsAsset /BRCosmetics/Accessories/FORT_Backpacks/Backpack_F_MED_Toxic_Kitty/F_MED_Toxic_Kitty_Pack_Physics
bodies {
 dyn_baton_a [{shapeType=Capsule, length=10.463, radius=0.8638, center=<4.7692 0.0001 0>, rotation=<rot 90 0 0>}]
 dyn_baton_a2 [{shapeType=Capsule, length=10.463, radius=0.8638, center=<4.494 0 0>, rotation=<rot 90 0 0>}]
 dyn_baton_b [{shapeType=Capsule, length=11.5802, radius=1.3268, center=<6.2534 0.0001 0>, rotation=<rot 90 0 0>}]
 dyn_baton_b2 [{shapeType=Capsule, length=10, radius=1.3268, center=<6.3611 0 0>, rotation=<rot 90 0 0>}]
 dyn_baton_mstr [{shapeType=Box, extentX=1.1, extentY=1.1, extentZ=1.1}]
 dyn_mask [{shapeType=Box, extentX=21.5763, extentY=4.1471, extentZ=19.0739, center=<8.3321 -1.3855 -0.2968>}]
 dyn_mask_mstr [{shapeType=Box, extentX=1.1, extentY=1.1, extentZ=1.1}]
 dyn_strap_l [{shapeType=Box, extentX=1.1, extentY=1.1, extentZ=5, center=<2 0 0>, rotation=<rot 90 0 0>}]
 dyn_strap_l_mstr [{shapeType=Box, extentX=1.1, extentY=1.1, extentZ=1.1}]
 dyn_strap_r [{shapeType=Box, extentX=1.1, extentY=1.1, extentZ=5, center=<-2 0 0>, rotation=<rot 90 0 0>}]
 dyn_strap_r_mstr [{shapeType=Box, extentX=1.1, extentY=1.1, extentZ=1.1}]
 pack [{shapeType=Capsule, length=21.737, radius=5.7056}, {shapeType=Box, extentX=27, extentY=26, extentZ=25.448, center=<0 -1.0167 -5.842>}]
}
constraints [
 {bone1Name=dyn_baton_a, bone2Name=dyn_baton_mstr, swing1LimitDegrees=90, swing2LimitDegrees=65, twistLimitDegrees=10}
 {bone1Name=dyn_baton_b, bone2Name=dyn_baton_mstr, swing1LimitDegrees=90, swing2LimitDegrees=65, twistLimitDegrees=10}
 {
  bone1Name=dyn_mask
  bone2Name=dyn_mask_mstr
  swing1LimitDegrees=45
  swing1Motion=Free
  swing2LimitDegrees=45
  swing2Motion=Free
  twistLimitDegrees=45
  twistMotion=Free
 }
 {bone1Name=dyn_strap_l, bone2Name=dyn_strap_l_mstr, swing1LimitDegrees=0, swing2LimitDegrees=45, twistLimitDegrees=0}
 {bone1Name=dyn_strap_r, bone2Name=dyn_strap_r_mstr, swing1LimitDegrees=0, swing2LimitDegrees=45, twistLimitDegrees=0}
 {bone1Name=dyn_baton_a2, bone2Name=dyn_baton_a, swing1LimitDegrees=90, swing2LimitDegrees=65, twistLimitDegrees=10}
 {bone1Name=dyn_baton_b2, bone2Name=dyn_baton_b, swing1LimitDegrees=90, swing2LimitDegrees=65, twistLimitDegrees=10}
]
