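# Post-process style animation blueprint for the Toxic Kitty backpack cosmetic, adding secondary motion to its dangling bones on top of the input pose.
# At LOD 0-1 it runs a RigidBody simulation in root-bone space.
# At LOD 2+, set on each update, it blends to two planar-limited baton trail controllers feeding a partial mask rotation constraint.
# A Flap node for the mask is present but unconnected.
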AnimBlueprint /BRCosmetics/Accessories/FORT_Backpacks/Backpack_F_MED_Toxic_Kitty/F_MED_Toxic_Kitty_Pack_Skeleton_AnimBlueprint
target_skeleton=&/F_MED_Toxic_Kitty_Pack_Skeleton
graphs [
 AnimGraph {nodes [
   n1 = RigidBody(ComponentPose=Local To Component(LocalPose=Use cached pose 'InputPose'())) {
    bClampLinearTranslationLimitToRefPose=true
    bOverrideWorldGravity=true
    overridePhysicsAsset=&/F_MED_Toxic_Kitty_Pack_Physics
    simulationSpace=BaseBoneSpace
    alphaBoolBlend {blendOption=Linear}
    baseBoneRef {boneName=root}
    componentAppliedLinearAccClamp=<5000 5000 5000>
    componentLinearAccScale=<0.7 0.7 0.7>
    componentLinearVelScale=<0.7 0.7 0.7>
    overrideWorldGravity {z=-980}
   }
   n2 = Trail controller - Bone: dyn_baton_a2(ComponentPose=Local To Component(LocalPose=Use cached pose 'InputPose'()), Alpha=1) {
    bUsePlanarLimit=true
    relaxationSpeedScale=5.2154
    stretchLimit=-1.6084
    baseJoint {boneName=pack}
    planarLimits [{drivingBone {boneName=pack}, planeTransform {location=<0 0 -20>, rotation=<rot 0 0 0>, scale=<1 1 1>}}]
    rotationLimits [{limitMax=<180 180 180>}, {limitMax=<180 180 180>}]
    rotationOffsets [<0 0 0>, <0 0 0>]
    trailBone {boneName=dyn_baton_a2}
    bLimitStretch=~
   }
   n3 = Trail controller - Bone: dyn_baton_b2(ComponentPose=@n2, Alpha=1) {
    bUsePlanarLimit=true
    relaxationSpeedScale=6.2546
    baseJoint {boneName=pack}
    planarLimits [{drivingBone {boneName=pack}, planeTransform {location=<0 0 -20>, rotation=<rot 0 0 0>, scale=<1 1 1>}}]
    rotationLimits [{limitMax=<180 180 180>}, {limitMax=<180 180 180>}]
    rotationOffsets [<0 0 0>, <0 0 0>]
    trailBone {boneName=dyn_baton_b2}
    bLimitStretch=~
   }
   n4 = "Constraint - dyn_mask (R)"(ComponentPose=@n3, Alpha=0.78, ConstraintWeights_0=1) {
    alphaBoolBlend {blendOption=Linear}
    boneToModify {boneName=dyn_mask}
    constraintSetup [{transformType=Rotation, targetBone {boneName=dyn_baton_a}}]
    constraintWeights [1]
   }
   Output Pose(Result=Blend Poses by bool(
     BlendPose_0=Component To Local(ComponentPose=@n4)
     BlendPose_1=Component To Local(ComponentPose=@n1)
     bActiveValue=Get LODIs2+() -> (LODIs2+:bool)
    ) {blendType=Linear}) {layerGroup=None}
   AnimGraphNode_SubInput_0 = Input Pose()
   Save cached pose 'InputPose'(Pose=@AnimGraphNode_SubInput_0) {cacheName=InputPose}
   Flap Node - Bone: dyn_mask!(ComponentPose:Component Space Pose Link Structure, Alpha:float=1) -> (Pose:Component Space Pose Link Structure) {
    is=title FortAnimGraphNode_Flap
    targetComponent=RotationZ
    alphaBoolBlend {blendOption=Linear}
    sourceDrives [{
      bUseRange=false
      drivingCurve=null
      multiplier=1.3662
      rangeMax=1
      rangeMin=-1
      remappedMax=1
      remappedMin=0
      sourceComponent=RotationZ
      useQuaternion=true
      sourceBone {boneName=pack}
     }]
    targetBone {boneName=dyn_mask}
   }
  ]}
 EventGraph {nodes [
   Event BlueprintUpdateAnimation()
   > n1 = Branch(Condition="integer > integer"(A=Get CurrentLOD(), B=1) {is=title K2Node_CallFunction})
   > Set LODIs2+(LODIs2+:bool=true) -> (Output_Get:bool)
   Set LODIs2+(execute=@n1.else, LODIs2+=false)
   TryGetPawnOwner()
  ]}
]
variables [LODIs2+ {type=boolean}]
